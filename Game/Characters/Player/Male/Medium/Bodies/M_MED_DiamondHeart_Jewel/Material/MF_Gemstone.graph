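# Material function for the DiamondHeart Jewel player body that blends a faked gemstone look into incoming material attributes by TechArtMask.
# A refracted ray samples a distance-corrected cubemap of random normals to build interior facet normals.
# These feed sky reflection/refraction with a total-internal-reflection term, custom GGX sun glints (cheaper on mobile), scattering and thin-film dispersion.
# Mask-texture channels add AO tint and curvature, while time-of-day functions scale the emissive.
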MaterialFunction /Game/Characters/Player/Male/Medium/Bodies/M_MED_DiamondHeart_Jewel/Material/MF_Gemstone
graph {nodes [
  ActorPositionWS()
  n1 = MaterialFunctionCall() {materialFunction=&ObjectPivotPoint}
  n2 = VectorParameter() {group=Gemstone, parameterName=GS_CubeMapPositionOffset, defaultValue=<rgba 0 0 0 1>}
  Add(A=@n1.Object Pivot Location, B=@n2.RGB)
  Add(A=ShadingPathSwitch(Default=Constant(), Mobile=NamedRerouteUsage()), B=NamedRerouteUsage()) {constB=0}
  Constant3Vector() {constant=<rgba 1 0 0.6367 1>}
  n3 = VectorParameter() {group=Gemstone, parameterName=GS_ColorTint, sortPriority=-2, defaultValue=<rgba 0.7623 0.8772 0.9323 1>}
  n4 = LinearInterpolate(
   A=VectorParameter() {group=Gemstone, parameterName=InColor, defaultValue=<rgba 0 0 0 1>}.RGB
   B=@n3.RGB
   Alpha=ScalarParameter() {defaultValue=1, group=Gemstone, parameterName=GS_BaseColorBlend, sortPriority=-1}
  )
  Desaturation(None=@n4, Fraction=Constant() {r=-0.2})
  n5 = FunctionInput() {inputName=Material, inputType=FunctionInput_MaterialAttributes}
  n6 = GetMaterialAttributes(None=@n5) {attributeGetTypes [#1, #2, #3]}
  n7 = ScalarParameter() {defaultValue=1, group=Gemstone, parameterName=GS_Specular}
  n8 = SetMaterialAttributes!(
   MaterialAttributes=@n5
   Metallic=ScalarParameter() {defaultValue=1, group=Gemstone, parameterName=GS_Metallic}
   Roughness=Multiply(A=@n6.Roughness, B=ScalarParameter() {defaultValue=0.1, group=Gemstone, parameterName=GS_Roughness})
   Specular=Saturate(None=Multiply(A=@n6.Specular, B=@n7))
   EmissiveColor=NamedRerouteUsage()
   Opacity=ScalarParameter() {defaultValue=0.7, group=Gemstone, parameterName=GS_Opacity}
   SubsurfaceColor=NamedRerouteUsage()
   BaseColor=NamedRerouteUsage()
   ShadingModel=ShadingModel()
  ) {attributeSetTypes [
    #4
    #1
    #3
    #5
    #6
    #7
    #8
    #9
   ]}
  n9 = Constant() {r=1}
  FunctionOutput(None=BlendMaterialAttributes(A=@n5, B=@n8, Alpha=FunctionInput(Preview=@n9) {inputName=TechArtMask, inputType=FunctionInput_Scalar, sortPriority=2}))
  n10 = LinearInterpolate(
   A=ScalarParameter() {defaultValue=1, group=Gemstone, parameterName=GS_GradientLow}
   B=ScalarParameter() {defaultValue=0.1, group=Gemstone, parameterName=GS_GradientHigh}
   Alpha=DotProduct(A=PixelNormalWS(), B=CameraVectorWS())
  ) {constA=0.45}
  LinearInterpolate(A=@n10, Alpha=NamedRerouteUsage())
  n11 = NamedRerouteUsage()
  n12 = ScalarParameter() {defaultValue=100, group=Gemstone, parameterName=GS_RefractedCubeMapRadius_Near}
  n13 = ScalarParameter() {defaultValue=100, group=Gemstone, parameterName=GS_RefractedCubeMapRadius_Far}
  n14 = NamedRerouteUsage()
  n15 = ScalarParameter() {defaultValue=1.35, group=Gemstone, parameterName=GS_IOR_Near}
  n16 = LinearInterpolate(A=@n15, B=ScalarParameter() {defaultValue=1.35, group=Gemstone, parameterName=GS_IOR_Far}, Alpha=NamedRerouteUsage())
  n17 = MaterialFunctionCall!(
   CubePosition=TransformPosition(None=@n2.RGB) {transformType=TRANSFORMPOSSOURCE_World}
   Radius=StaticSwitchParameter(True=LinearInterpolate(A=@n12, B=@n13, Alpha=@n14), False=@n12) {group=Gemstone, parameterName=GS_UseCameraFade}
   UseRefractedRay=StaticBool() {value=true}
   IOR=NamedRerouteDeclaration(None=StaticSwitchParameter(True=@n16, False=@n15) {group=Gemstone, parameterName=GS_UseCameraFade}) {name=IOR, nodeColor=<rgba 1 0 0.2109 1>}
   CustomNormalWS=PixelNormalWS()
   UsePreskinnedPosition=StaticBoolParameter() {group=Gemstone, parameterName=GS_Cubemap_UsePreskinnedPos}
  ) {materialFunction=/Game/Global/MaterialFunctions/TechArt/MF_DistanceCorrectedCubemapSample}
  n18 = TextureSampleParameterCube!(UVs=NamedRerouteDeclaration(None=@n17.DistanceCorrectedRay) {name=UVW, nodeColor=<rgba 0 0.625 1 1>}, Bias, Apply View MipBias) {
   group=Gemstone
   mipValueMode=TMVM_MipBias
   parameterName=GS_RefractionCubeMap
   samplerType=SAMPLERTYPE_LinearColor
   texture=/Game/Global/Textures/Cubemaps/TC_Normals_Random_S
  }
  n19 = Normalize(VectorInput=ConstantBiasScale(None=@n18.RGB) {bias=-0.5, scale=2})
  n20 = NamedRerouteDeclaration(None=LinearInterpolate(
    A=PixelNormalWS()
    B=Multiply(A=@n19) {constB=-1}
    Alpha=NamedRerouteDeclaration(None=@n17.Mask) {name=CubeMask, nodeColor=<rgba 0 1 0.6016 1>}
   )) {name=InteriorNormal, nodeColor=<rgba 0.5234 1 0 1>}
  n21 = NamedRerouteUsage()
  n22 = Constant() {r=1}
  NamedRerouteDeclaration(None=Multiply(A=MaterialFunctionCall!(Ray Direction=@n11, Surface Normal=@n20, Refractive Index Origin=@n21, Refractive Index Target=@n22) {materialFunction=/Engine/Functions/Engine_MaterialFunctions01/Vectors/Refract}.Vector) {constB=-1}) {name=RefractVectorInterior, nodeColor=<rgba 1 0.9375 0 1>}
  MaterialExpressionNormalize_2 = Normalize(VectorInput=SkyAtmosphereLightDirection())
  MaterialExpressionCameraVectorWS_1 = CameraVectorWS()
  MaterialExpressionNamedRerouteUsage_8 = NamedRerouteUsage()
  MaterialExpressionConstant_28 = Constant() {r=0.3}
  MaterialExpressionConstant3Vector_3 = Constant3Vector() {constant=<rgba 1 1 1 0>}
  n23 = Custom!(
   L=@MaterialExpressionNormalize_2
   V=@MaterialExpressionCameraVectorWS_1
   N=@MaterialExpressionNamedRerouteUsage_8
   roughness=@MaterialExpressionConstant_28
   F0=@MaterialExpressionConstant3Vector_3
  ) {
   code="    float3 H = normalize(L + V);\r\n    float NdotH = saturate(dot(N, H));\r\n    float NdotL = saturate(dot(N, L));\r\n    float NdotV = saturate(dot(N, V));\r\n    float3 F = F0 + (1.0 - F0) * pow(1.0 - NdotV, 5.0);\r\n    float epsilon = .0001;\r\n\r\n\tfloat alpha = roughness * roughness;\r\n    float alphaSqr = alpha * alpha;\r\n\r\n    float denominator = NdotH * NdotH * (alphaSqr - 1.0) + 1.0;\r\n    float D = alphaSqr / (PI * denominator * denominator + epsilon);\r\n\t\r\n\t\r\n\t\r\n    float k = alpha / 2;    \r\n    float g_v = NdotV / ( NdotV*(1 - k ) + k);\r\n    float g_l = NdotL / (NdotL*(1 - k ) + k);\r\n    float G = g_v * g_l;\r\n    return NdotL*(D * G * F) / (4.0 * NdotL * NdotV + epsilon );"
   description=GGX
   inputs [
    {
     inputName=L
     input {expression=&self:MaterialExpressionNormalize_2}
    }
    {
     inputName=V
     input {expression=&self:MaterialExpressionCameraVectorWS_1}
    }
    {
     inputName=N
     input {expression=&self:MaterialExpressionNamedRerouteUsage_8}
    }
    {
     inputName=roughness
     input {expression=&self:MaterialExpressionConstant_28}
    }
    {
     inputName=F0
     input {
      expression=&self:MaterialExpressionConstant3Vector_3
      mask=1
      maskB=1
      maskG=1
      maskR=1
     }
    }
   ]
  }
  n24 = Multiply(A=@n23) {constA=1}
  n25 = DotProduct(A=NamedRerouteUsage(), B=@MaterialExpressionNormalize_2)
  n26 = Multiply(A=Power(Base=Saturate(None=@n25)) {constExponent=24}) {constB=5}
  n27 = NamedRerouteUsage()
  n28 = Multiply(A=Clamp(None=ShadingPathSwitch(Default=@n24, Mobile=@n26)) {maxDefault=40}, B=Multiply(A=@n27, B=@n27))
  n29 = NamedRerouteUsage()
  n30 = NamedRerouteUsage()
  NamedRerouteDeclaration(None=Multiply(A=Multiply(A=@n28, B=@n29), B=LinearInterpolate(Alpha=@n30) {constA=0.5, constB=0.05})) {name=SunInternalReflections, nodeColor=<rgba 0 0.8828 1 1>}
  n31 = NamedRerouteUsage()
  n32 = SkyLightEnvMapSample(Direction=NamedRerouteUsage(), Roughness=ScalarParameter() {defaultValue=0.1, group=Gemstone, parameterName=GS_ReflectRoughness})
  n33 = SkyLightEnvMapSample(Direction=NamedRerouteUsage(), Roughness=ScalarParameter() {defaultValue=0.2, group=Gemstone, parameterName=GS_RefractRoughness})
  n34 = Multiply(A=@n33) {constB=1.2}
  n35 = NamedRerouteUsage()
  n36 = Multiply(
   A=LinearInterpolate(A=Multiply(A=@n32) {constB=0.75}, B=Multiply(A=@n34, B=@n35), Alpha=NamedRerouteUsage())
   B=ScalarParameter() {defaultValue=1, group=Gemstone, parameterName=GS_ReflectionsBrightness}
  )
  n37 = Desaturation(
   None=Power(Base=Saturate(None=@n36), Exp=ScalarParameter() {defaultValue=1.1, group=Gemstone, parameterName=GS_RefractionsExponent})
   Fraction=ScalarParameter() {defaultValue=0.1, group=Gemstone, parameterName=GS_RefractionsDesaturation}
  )
  NamedRerouteDeclaration(None=Multiply(A=NamedRerouteUsage(), B=Multiply(A=@n31, B=@n37)) {constA=1}) {name=Sky Refractions, nodeColor=<rgba 0.6875 1 0 1>}
  NamedRerouteDeclaration(None=@n18.RGB) {name=DebugNormal, nodeColor=<rgba 0.6875 0 1 1>}
  n38 = Constant() {r=1}
  n39 = ComponentMask(None=NamedRerouteUsage()) {b=true}
  n40 = Saturate(None=OneMinus(None=Multiply(A=@n39) {constB=-1}))
  NamedRerouteDeclaration(None=MaterialFunctionCall!(1=@n38, 2=Constant() {r=0.5}, 3=@n38, 4, A=Multiply(A=@n40, B=@n40)) {materialFunction=/Engine/Functions/Engine_MaterialFunctions02/Lerp_Multiple_Float}.Lerp 3 Inputs) {name=GroundGradient, nodeColor=<rgba 0.5703 0 1 1>}
  n41 = Multiply(
   A=Multiply(A=NamedRerouteUsage()) {constB=0.5}
   B=LinearInterpolate(
    A=NamedRerouteUsage()
    B=ScalarParameter() {defaultValue=0.45, group=Gemstone, parameterName=GS_InteriorDiffuse_B}
    Alpha=NamedRerouteUsage()
   ) {constA=1, constB=0}
  )
  n42 = NamedRerouteUsage()
  n43 = SkyAtmosphereLightDirection()
  n44 = LinearInterpolate(A=ScalarParameter() {defaultValue=0.8, group=Gemstone, parameterName=GS_InteriorDiffuse_A}, Alpha=Abs(None=DotProduct(A=@n42, B=@n43))) {constA=0.25}
  NamedRerouteDeclaration(None=Multiply(A=Multiply(A=@n41, B=@n44), B=ScalarParameter() {defaultValue=1.5, group=Gemstone, parameterName=GS_InteriorDiffuse_Strength})) {name=InteriorDiffuse, nodeColor=<rgba 1 0 0.5859 1>}
  n45 = Add(A=NamedRerouteUsage(), B=NamedRerouteUsage())
  n46 = NamedRerouteUsage()
  n47 = ScalarParameter() {defaultValue=0.748, group=Gemstone, parameterName=GS_CoreDarkening}
  n48 = ScalarParameter() {defaultValue=0.3, group=Gemstone, parameterName=GS_CoreColor}
  NamedRerouteDeclaration(None=Multiply(A=Add(A=@n45, B=@n46), B=Fresnel(ExponentIn=@n47, BaseReflectFractionIn=@n48) {baseReflectFraction=0.1})) {name=Emissive, nodeColor=<rgba 1 0.9609 0 1>}
  NamedRerouteDeclaration(None=@n6.Normal) {name=NormalMap, nodeColor=<rgba 0.9219 1 0 1>}
  n49 = PixelNormalWS()
  n50 = CameraVectorWS()
  n51 = SphereMask(
   A=Power(Base=DotProduct(A=@n49, B=@n50)) {constExponent=5}
   B=ScalarParameter() {defaultValue=0.2, group=Gemstone, parameterName=GS_Diffusion_Ring}
  ) {attenuationRadius=0.2, hardnessPercent=0}
  n52 = NamedRerouteUsage()
  n53 = TextureSample(UVs=Multiply(A=ComponentMask(None=@n52) {g=true, r=true}, B=ScalarParameter() {defaultValue=2, group=Gemstone, parameterName=GS_Diffusion_Tiling}) {constB=4}) {texture=/Game/Global/Textures/Default/Thinfilm/T_ThinFilm_Spectrum_COLOR}
  NamedRerouteDeclaration(None=Desaturation(None=Multiply(A=@n51, B=@n53.RGB), Fraction=ScalarParameter() {group=Gemstone, parameterName=GS_Diffusion_Desaturation})) {name=ColorDiffusion, nodeColor=<rgba 1 0 0.9609 1>}
  n54 = NamedRerouteUsage()
  n55 = NamedRerouteUsage()
  n56 = Add(A=NamedRerouteUsage(), B=Multiply(A=Multiply(A=@n54, B=@n55)) {constB=0.5})
  n57 = VectorParameter() {group=Gemstone, parameterName=GS_AO_Color, defaultValue=<rgba 0.0099 0.0164 0.1354 1>}
  n58 = NamedRerouteUsage()
  NamedRerouteDeclaration(None=StaticSwitchParameter(True=LinearInterpolate(A=@n56, B=@n57.RGB, Alpha=@n58), False=@n56) {group=Gemstone, parameterName=GS_AO_Enable}) {name=Base Color, nodeColor=<rgba 0.7812 1 0 1>}
  NamedRerouteDeclaration(None=MaterialFunctionCall(
    Fade Length=ScalarParameter() {defaultValue=512, group=Gemstone, parameterName=GS_Distance_FadeLength}
    Fade Offset=ScalarParameter() {defaultValue=450, group=Gemstone, parameterName=GS_Distance_FadeOffset}
   ) {materialFunction=&CameraDepthFade}.Result) {name=CameraFade, nodeColor=<rgba 1 0.2344 0 1>}
  n59 = TextureSample() {
   samplerType=SAMPLERTYPE_Masks
   texture=/Game/Characters/Player/Male/Medium/Bodies/M_MED_DiamondHeart_Jewel/Textures/T_DiamondHeart_Jewel_Body_M
  }
  n60 = MaterialFunctionCall(Float3=NamedRerouteDeclaration(None=FunctionInput(Preview=@n59.RGB) {inputName=Mask Texture, sortPriority=1}) {name=Mask, nodeColor=<rgba 1 0 0.0469 1>}) {materialFunction=&BreakOutFloat3Components}
  n61 = NamedRerouteDeclaration(None=@n60.R) {name=AO, nodeColor=<rgba 1 0.6797 0 1>}
  n62 = Multiply(
   A=Power(Base=OneMinus(None=@n61), Exp=ScalarParameter() {defaultValue=2, group=Gemstone, parameterName=GS_AO_Power}) {constExponent=5}
   B=ScalarParameter() {defaultValue=2, group=Gemstone, parameterName=GS_AO_Intensity}
  ) {constB=3}
  n63 = SmoothStep(Value=NamedRerouteUsage()) {constMax=0, constMin=0.5}
  n64 = LinearInterpolate(A=@n62, B=Add(A=@n62, B=Multiply(A=@n63) {constB=0.5}), Alpha=ScalarParameter() {group=Gemstone, parameterName=GS_AO_UseCurvature_MaskB})
  NamedRerouteDeclaration(None=StaticSwitchParameter(True=Saturate(None=@n64), False=Constant()) {group=Gemstone, parameterName=GS_AO_Enable}) {name=Ambient Occlusion, nodeColor=<rgba 0.0078 1 0 1>}
  n65 = PixelNormalWS()
  n66 = CameraVectorWS()
  n67 = Saturate(None=DotProduct(
    A=MaterialFunctionCall!(Normal=@n65, CameraVector=@n66) {materialFunction=/Engine/Functions/Engine_MaterialFunctions02/Utility/CustomReflectionVector}.Reflection Vector
    B=SkyAtmosphereLightDirection()
   ))
  n68 = Multiply(A=SmoothStep(Value=NamedRerouteUsage()) {constMin=0.5}, B=Multiply(A=Power(Base=@n67) {constExponent=6}) {constB=5})
  n69 = VertexNormalWS()
  n70 = ComponentMask(None=@n69) {b=true}
  n71 = Saturate(None=SmoothStep(Value=SquareRoot(None=@n70)) {constMin=0.7})
  n72 = Saturate(None=DotProduct(A=@n69, B=SkyAtmosphereLightDirection()))
  NamedRerouteDeclaration(None=Multiply(A=Multiply(A=@n68), B=Multiply(A=@n71, B=@n72))) {name=Curvature Specular Highlights, nodeColor=<rgba 0 0.25 1 1>}
  n73 = Multiply(A=NamedRerouteUsage(), B=ScalarParameter() {defaultValue=0.3, group=Gemstone, parameterName=GS_DiffusionColorAdd})
  n74 = Multiply(A=@n73) {constB=0.8}
  n75 = MaterialFunctionCall(Color_Morning=@n73, Color_Day=@n73, Color_Evening=@n74, Color_Night=@n74) {materialFunction=&Color_Time_of_Day}
  n76 = NamedRerouteDeclaration(None=@n60.G) {name=Gmap, nodeColor=<rgba 0 1 0.6719 1>}
  n77 = Multiply(A=Constant() {r=1.2}, B=NamedRerouteUsage())
  n78 = NamedRerouteUsage()
  n79 = ScalarParameter() {defaultValue=3, group=Gemstone, parameterName=GS_SSS_Intensity}
  n80 = Divide(A=@n79) {constB=4}
  n81 = Desaturation(
   None=Multiply(
    A=Multiply(A=@n77, B=@n78)
    B=MaterialFunctionCall(Value_Morning=@n79, Value_Day=@n79, Value_Evening=@n80, Value_Night=@n80) {materialFunction=&Scalar_Time_of_Day}.Result
   ) {constB=5}
   Fraction=Constant() {r=-2}
  )
  n82 = Multiply(
   A=Multiply(A=LinearInterpolate(A=@n76), B=Add(A=@n77, B=@n81))
   B=ScalarParameter() {defaultValue=2, group=Gemstone, parameterName=GS_BaseIntensity}
  )
  n83 = Multiply(A=@n82) {constB=3}
  n84 = MaterialFunctionCall(Color_Morning=@n82, Color_Day=@n82, Color_Evening=@n83, Color_Night=@n83) {materialFunction=&Color_Time_of_Day}
  n85 = NamedRerouteUsage()
  n86 = NamedRerouteUsage()
  n87 = Multiply(
   A=Add(A=Add(A=@n75.Result, B=@n84.Result), B=Multiply(A=@n85, B=@n86)) {constB=0}
   B=ScalarParameter() {defaultValue=1, group=Gemstone, parameterName=GD_Overall_Intensity}
  )
  n88 = LinearInterpolate(
   A=@n87
   B=VectorParameter() {group=Gemstone, parameterName=GS_AO_Color, defaultValue=<rgba 0.0099 0.0164 0.1354 1>}.RGB
   Alpha=NamedRerouteUsage()
  )
  NamedRerouteDeclaration(None=StaticSwitchParameter(True=NamedRerouteUsage(), False=StaticSwitchParameter(True=@n88, False=@n87) {group=Gemstone, parameterName=GS_AO_Enable}) {group=Gemstone, parameterName=GS_Debug_Normal}) {name=Emissive Color, nodeColor=<rgba 0 0.2969 1 1>}
  NamedRerouteDeclaration(None=@n33) {name=Debug, nodeColor=<rgba 0 0.1094 1 1>}
  NamedRerouteDeclaration(None=@n60.B) {name=Curvature, nodeColor=<rgba 1 0 0.6328 1>}
  n89 = PixelNormalWS()
  n90 = CameraVectorWS()
  n91 = Multiply(A=ScalarParameter() {defaultValue=1, group=Crystal, parameterName=BacklightViewAngle, sortPriority=16}, B=@n90)
  n92 = DotProduct(A=Normalize(VectorInput=Add(A=@n89, B=@n91)), B=Multiply(A=SkyAtmosphereLightDirection()) {constB=-1})
  n93 = Saturate(None=Add(A=Multiply(A=@n92) {constB=0.5}) {constB=0.5})
  n94 = ScalarParameter() {defaultValue=2, group=Crystal, parameterName=BacklightAperture, sortPriority=17}
  n95 = DotProduct(A=@n89, B=@n90)
  NamedRerouteDeclaration(None=Multiply(A=Power(Base=@n93, Exp=@n94), B=Saturate(None=@n95))) {name=Interior Light Scattering, nodeColor=<rgba 0.5 1 0 1>}
  n96 = Multiply(A=DotProduct(A=NamedRerouteUsage(), B=NamedRerouteUsage())) {constB=-1}
  n97 = ScalarParameter() {defaultValue=0.35, group=Gemstone, parameterName=GS_TIR_Min}
  n98 = ScalarParameter() {defaultValue=0.8, group=Gemstone, parameterName=GS_TIR_Max}
  NamedRerouteDeclaration(None=LinearInterpolate(
    A=ScalarParameter() {defaultValue=0.3, group=Gemstone, parameterName=GS_ReflectionAngleMin}
    Alpha=MaterialFunctionCall(Input=@n96, Input Low=@n97, Input High=@n98) {materialFunction=&RemapValue_0-1}.Result
   ) {constA=0.5}) {name=TIR, nodeColor=<rgba 0 0.7422 1 1>}
  NamedRerouteDeclaration(None=@n17.RefractedRay) {name=RefractedRay, nodeColor=<rgba 1 0 0.8906 1>}
  NamedRerouteDeclaration(None=@n10) {name=Depth Gradient, nodeColor=<rgba 1 0 0.2812 1>}
  NamedRerouteDeclaration(None=@n4) {name=Tint, nodeColor=<rgba 0.3125 1 0 1>}
  NamedRerouteDeclaration(None=ReflectionVectorWS(CustomWorldNormal=NamedRerouteUsage())) {name=ReflectionVector, nodeColor=<rgba 1 0 0.3281 1>}
  n99 = Constant3Vector() {constant=<rgba 0 0 100 0>}
  NamedRerouteDeclaration(None=Multiply(A=SkyAtmosphereLightIlluminance!(World Position=@n99)) {constB=0.5}) {name=DirectionalLightColor, nodeColor=<rgba 0 0.4844 1 1>}
  n100 = DotProduct(A=SkyAtmosphereLightDirection(), B=NamedRerouteUsage())
  n101 = ScalarParameter() {defaultValue=0.6, group=Gemstone, parameterName=GS_Phase}
  n102 = Clamp(None=Multiply(
    A=NamedRerouteUsage()
    B=MaterialFunctionCall!(Cosine=@n100, Anisotropy=@n101) {materialFunction=/Engine/Functions/Engine_MaterialFunctions01/Volumetrics/SchlickPhase}.Result
   )) {maxDefault=20}
  n103 = NamedRerouteUsage()
  n104 = NamedRerouteUsage()
  NamedRerouteDeclaration(None=Multiply(A=Multiply(A=@n102, B=@n103) {constB=2}, B=LinearInterpolate(Alpha=@n104) {constA=0.4})) {name=SunNscatter, nodeColor=<rgba 1 0.7031 0 1>}
  NamedRerouteUsage()
  NamedRerouteUsage()
  NamedRerouteUsage()
  NamedRerouteUsage()
  NamedRerouteUsage()
  NamedRerouteUsage()
  NamedRerouteUsage()
  NamedRerouteUsage()
  NamedRerouteUsage()
  NamedRerouteUsage()
  StaticSwitchParameter(True=@n1.Object Pivot Location, False=VectorParameter() {group=Gemstone, parameterName=GS_CubeMapPosition, defaultValue=<rgba 0 0 0 1>}.RGB) {defaultValue=true, group=Gemstone, parameterName=GS_CubemapFixedPosition}
  Transform(None=NamedRerouteUsage()) {transformSourceType=TRANSFORMSOURCE_World, transformType=TRANSFORM_Tangent}
  VectorParameter() {group=Gemstone, parameterName=GS_CubemapLocalPosition, defaultValue=<rgba 0 0 0 1>}
 ]}
