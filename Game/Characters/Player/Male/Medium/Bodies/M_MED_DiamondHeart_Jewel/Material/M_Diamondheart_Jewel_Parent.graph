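# Parent material for the DiamondHeart Jewel male player body, adding a gemstone look and optional glowing eyes to the 2023 character shading stack.
# Character PreFX output passes through MF_Gemstone, masked by a selectable channel from either the Tech Art Mask or a custom mask texture.
# A static switch optionally blends Base Color times an HDR eye color into Emissive through an eye channel mask, before the PostFX function finishes the material.
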Material /Game/Characters/Player/Male/Medium/Bodies/M_MED_DiamondHeart_Jewel/Material/M_Diamondheart_Jewel_Parent
graph {
 properties {
  bUseMaterialAttributes=true
  bUsedWithMeshDeformer=true
  bUsedWithMorphTargets=true
  bUsedWithSkeletalMesh=true
  pixelDepthOffsetMode=PDOM_Legacy
  shadingModel=MSM_FromMaterialExpression
  usedShadingModels=Default Lit | Subsurface | Clear Coat | Subsurface Profile | Cloth
 }
 nodes [
  n1 = MaterialFunctionCall() {materialFunction=&MF_CharacterParent_2023_PreFX}
  n2 = TextureSampleParameter2D() {
   group=DiamondHeart
   parameterName=DH_CustomMaskTexture
   samplerType=SAMPLERTYPE_Masks
   texture=&T_BlackMask
  }
  n3 = StaticSwitchParameter(True=@n2.RGB, False=@n1.Tech Art Mask) {group=DiamondHeart, parameterName=DH_UseCustomFXMask}
  n4 = MaterialFunctionCall!(
   Material=@n1.Result
   Mask Texture=@n1.Result
   TechArtMask=ChannelMaskParameter(None=@n3) {group=DiamondHeart, parameterName=DH_DiamondMaskChannel}.R
  ) {materialFunction=&/MF_Gemstone}
  n5 = GetMaterialAttributes(None=@n4.Result) {attributeGetTypes [#1, #2]}
  n6 = LinearInterpolate(
   A=@n5.Emissive Color
   B=Multiply(A=@n5.Base Color, B=VectorParameter() {group=DiamondHeart, parameterName=DH_Eye_Color, defaultValue=<rgba 0.6198 5.3128 10 1>}.RGB)
   Alpha=ChannelMaskParameter(None=@n3) {group=DiamondHeart, parameterName=DH_EyeMask}.R
  )
  MaterialFunctionCall(
   Material=StaticSwitchParameter(
    True=SetMaterialAttributes(MaterialAttributes=@n5.MaterialAttributes, Emissive Color=@n6) {attributeSetTypes [#1]}
    False=@n4.Result
   ) {group=DiamondHeart, parameterName=DH_UseEye}
   M=@n1.M
   Pre Skinned Local Position=@n1.M
   Distance Blend=@n1.M
   Fresnel=@n1.M
  ) {materialFunction=&MF_CharacterParent_2023_PostFX}
 ]
}
